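# Sound cue for applying a bandage in Athena, attenuated as a consumable use sound.
# A mixer layers one of four equally weighted random bandage apply waves with a quieter player health-up UI sound.
# It plays at reduced volume in the consumables sound class, with a concurrency override that stops the oldest instance when the limit is reached.
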SoundCue /Game/Sounds/Fort_GamePlay_Sounds/Player/Bandage_Apply_Athena_Cue
root=Mixer_5
nodes [
 {id=Mixer_5, children [Random_0, WavePlayer_38], properties {inputVolume [1,0.7]}}
 {id=Random_0, children [WavePlayer_3,WavePlayer_4,WavePlayer_5,WavePlayer_6], properties {weights [1,1,1,1]}}
 {id=WavePlayer_3, properties {soundWaveAssetPtr=&/Bandage_Apply_004}}
 {id=WavePlayer_38, properties {soundWaveAssetPtr=/Game/Sounds/Fort_UI_Sounds/FrontEnd/fort_ui_playerhealth_up_01}}
 {id=WavePlayer_4, properties {soundWaveAssetPtr=&/Bandage_Apply_002}}
 {id=WavePlayer_5, properties {soundWaveAssetPtr=&/Bandage_Apply_003}}
 {id=WavePlayer_6, properties {soundWaveAssetPtr=&/Bandage_Apply_001}}
]
properties {
 attenuationSettings=&Fort_Consumable_Use_Athena_Att
 bOverrideConcurrency=true
 maxDistance=2250
 soundClassObject=&Fort_Consumables
 volumeMultiplier=0.3
 concurrencyOverrides {maxCount=4, resolutionRule=StopOldest}
}
