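# Sound cue for eating a Flopper health consumable in Athena, played through the consumables sound class with a consumable-use attenuation.
# A random node picks one of three equally weighted eat waves and feeds a modulator with default settings.
# The modulator feeds a boosted input on the root mixer, whose second input is empty.
# A concurrency override caps simultaneous plays at two and stops the oldest.
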SoundCue /Game/Sounds/Fort_GamePlay_Sounds/Player/Durrable_HealthFlopper_Athena_Cue
root=Mixer_0
nodes [
 {id=Mixer_0, children [Modulator_0,null], properties {inputVolume [1.6,0.8]}}
 {id=Modulator_0, children [Random_0]}
 {id=Random_0, children [WavePlayer_1,WavePlayer_2,WavePlayer_4], properties {weights [1,1,1]}}
 {id=WavePlayer_1, properties {soundWaveAssetPtr=/Game/Athena/Sounds/Items/Flopper/Flopper_Eat2_01}}
 {id=WavePlayer_2, properties {soundWaveAssetPtr=/Game/Athena/Sounds/Items/Flopper/Flopper_Eat2_02}}
 {id=WavePlayer_4, properties {soundWaveAssetPtr=/Game/Athena/Sounds/Items/Flopper/Flopper_Eat2_03}}
]
properties {
 attenuationSettings=&Fort_Consumable_Use_Athena_Att
 bOverrideConcurrency=true
 maxDistance=2250
 soundClassObject=&Fort_Consumables
 volumeMultiplier=1.1
 concurrencyOverrides {maxCount=2, resolutionRule=StopOldest}
}
