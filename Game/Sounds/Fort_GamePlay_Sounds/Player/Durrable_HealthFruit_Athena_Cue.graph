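# Sound cue for using the Durrr Burger-style "Durrable" health fruit consumable in Athena, routed to the consumables sound class with consumable-use attenuation.
# A mixer layers the fruit wave, through a default modulator, with a health-pickup wave that a second modulator pitches up at fixed reduced volume.
# The mixer further lowers that health layer.
# A concurrency override limits simultaneous instances, stopping the oldest first.
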SoundCue /Game/Sounds/Fort_GamePlay_Sounds/Player/Durrable_HealthFruit_Athena_Cue
root=Mixer_0
nodes [
 {id=Mixer_0, children [Modulator_0,Modulator_1], properties {inputVolume [1,0.5]}}
 {id=Modulator_0, children [WavePlayer_1]}
 {id=Modulator_1, children [WavePlayer_0], properties {pitchMax=1.5, pitchMin=1.5, volumeMax=0.5, volumeMin=0.5}}
 {id=WavePlayer_0, properties {soundWaveAssetPtr=/Game/Sounds/Fort_Gadgets/Unassigned/new_health_get}}
 {id=WavePlayer_1, properties {soundWaveAssetPtr=&/Item_Durrable_Fruit_01}}
]
properties {
 attenuationSettings=&Fort_Consumable_Use_Athena_Att
 bOverrideConcurrency=true
 maxDistance=2250
 soundClassObject=&Fort_Consumables
 volumeMultiplier=0.9
 concurrencyOverrides {maxCount=4, resolutionRule=StopOldest}
}
