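# Sound cue for the player foley that plays when the blueprint is taken out as equipment.
# A random node picks between two takeout wave variants with equal weight.
# It routes to the player foley sound class with the generic player equip attenuation at reduced volume, and a concurrency override prevents new instances once the limit is reached.
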SoundCue /Game/Sounds/Fort_GamePlay_Sounds/Player/Player_Equip_Blueprint_Cue
root=Random_1
nodes [
 {id=Random_1, children [WavePlayer_0,WavePlayer_1], properties {weights [1,1]}}
 {id=WavePlayer_0, properties {soundWaveAssetPtr=&/player_blueprint_takeout_01}}
 {id=WavePlayer_1, properties {soundWaveAssetPtr=&/player_blueprint_takeout_02}}
]
properties {
 attenuationSettings=/Game/Sounds/Attenuation/Fort_Player_Equip_Generic_Att
 bOverrideConcurrency=true
 maxDistance=1800
 soundClassObject=&Fort_Player_Foley
 volumeMultiplier=0.3
 concurrencyOverrides {resolutionRule=PreventNew}
}
