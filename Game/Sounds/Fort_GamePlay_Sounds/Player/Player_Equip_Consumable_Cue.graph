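# Sound cue for the player foley heard when equipping a consumable.
# A Random node picks between two equally weighted equip waves and feeds a Modulator that fixes volume, leaving pitch at its default variation.
# It uses the generic player equip attenuation and the player foley sound class, with a slight volume boost.
# A concurrency override caps it at two simultaneous instances, stopping the oldest.
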SoundCue /Game/Sounds/Fort_GamePlay_Sounds/Player/Player_Equip_Consumable_Cue
root=Modulator_0
nodes [
 {id=Modulator_0, children [Random_0], properties {volumeMax=1, volumeMin=1}}
 {id=Random_0, children [WavePlayer_0,WavePlayer_1], properties {bShouldExcludeFromBranchCulling=true, weights [1,1]}}
 {id=WavePlayer_0, properties {soundWaveAssetPtr=&/Consumable_Equip_01}}
 {id=WavePlayer_1, properties {soundWaveAssetPtr=&/Consumable_Equip_02}}
]
properties {
 attenuationSettings=/Game/Sounds/Attenuation/Fort_Player_Equip_Generic_Att
 bOverrideConcurrency=true
 maxDistance=1800
 soundClassObject=&Fort_Player_Foley
 volumeMultiplier=1.2
 concurrencyOverrides {maxCount=2, resolutionRule=StopOldest}
}
