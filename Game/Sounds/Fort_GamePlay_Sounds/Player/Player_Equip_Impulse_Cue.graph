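# Sound cue for the player equip sound named Impulse, reusing a consumable grenade equip wave.
# A single wave player feeds a Modulator at its usual settings, with the whole cue pitched up.
# It plays in the player foley sound class with generic player equip attenuation.
# A concurrency override prevents new instances from starting once the limit is reached.
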SoundCue /Game/Sounds/Fort_GamePlay_Sounds/Player/Player_Equip_Impulse_Cue
root=Modulator_0
nodes [
 {id=Modulator_0, children [WavePlayer_2]}
 {id=WavePlayer_2, properties {soundWaveAssetPtr=&/Consumable_Grenade_Equip_05}}
]
properties {
 attenuationSettings=/Game/Sounds/Attenuation/Fort_Player_Equip_Generic_Att
 bOverrideConcurrency=true
 maxDistance=1800
 pitchMultiplier=1.5
 soundClassObject=&Fort_Player_Foley
 volumeMultiplier=1
 concurrencyOverrides {resolutionRule=PreventNew}
}
