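# Sound cue for the player equipping a large weapon, playing melee equip foley.
# A random node picks between two equally weighted equip wave variations.
# It routes to the player foley sound class with generic equip attenuation and slightly reduced volume.
# A concurrency override blocks new instances rather than cutting off ones already playing.
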SoundCue /Game/Sounds/Fort_GamePlay_Sounds/Player/Player_Equip_LargeWeapon_Cue
root=Random_0
nodes [
 {id=Random_0, children [WavePlayer_5,WavePlayer_6], properties {weights [1,1]}}
 {id=WavePlayer_5, properties {soundWaveAssetPtr=&/player_melee_equip_02}}
 {id=WavePlayer_6, properties {soundWaveAssetPtr=&/player_melee_equip_03}}
]
properties {
 attenuationSettings=/Game/Sounds/Attenuation/Fort_Player_Equip_Generic_Att
 bOverrideConcurrency=true
 maxDistance=1800
 soundClassObject=&Fort_Player_Foley
 volumeMultiplier=0.8
 concurrencyOverrides {resolutionRule=PreventNew}
}
