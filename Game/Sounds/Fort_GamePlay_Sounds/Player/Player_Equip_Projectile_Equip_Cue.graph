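# Player foley sound cue for equipping a projectile, reusing a melee equip sound wave.
# A single wave player feeds a Modulator that randomizes pitch and volume on each play.
# It uses the generic player equip attenuation and a concurrency override that prevents new instances once the limit is reached.
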SoundCue /Game/Sounds/Fort_GamePlay_Sounds/Player/Player_Equip_Projectile_Equip_Cue
root=Modulator_0
nodes [
 {id=Modulator_0, children [WavePlayer_5], properties {pitchMax=1.15, pitchMin=0.8, volumeMax=1.15, volumeMin=0.85}}
 {id=WavePlayer_5, properties {soundWaveAssetPtr=&/player_melee_equip_02}}
]
properties {
 attenuationSettings=/Game/Sounds/Attenuation/Fort_Player_Equip_Generic_Att
 bOverrideConcurrency=true
 maxDistance=1800
 soundClassObject=&Fort_Player_Foley
 volumeMultiplier=0.8
 concurrencyOverrides {resolutionRule=PreventNew}
}
