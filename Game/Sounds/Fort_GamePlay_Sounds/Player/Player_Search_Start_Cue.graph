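# Sound cue for the player starting a search, in the loot collect sound class.
# A random node picks one of four equally weighted search-start waves, and a modulator at the root varies its pitch slightly.
# It plays quietly through close-range mono prop attenuation.
# A concurrency override allows only one instance at a time, stopping the oldest.
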SoundCue /Game/Sounds/Fort_GamePlay_Sounds/Player/Player_Search_Start_Cue
root=Modulator_0
nodes [
 {id=Modulator_0, children [Random_2], properties {pitchMax=1.2, pitchMin=0.9}}
 {id=Random_2, children [WavePlayer_4, WavePlayer_9, WavePlayer_10, WavePlayer_11], properties {weights [1,1,1,1]}}
 {id=WavePlayer_10, properties {soundWaveAssetPtr=&/player_search_start_03}}
 {id=WavePlayer_11, properties {soundWaveAssetPtr=&/player_search_start_04}}
 {id=WavePlayer_4, properties {soundWaveAssetPtr=&/player_search_start_01}}
 {id=WavePlayer_9, properties {soundWaveAssetPtr=&/player_search_start_02}}
]
properties {
 attenuationSettings=/Game/Sounds/Attenuation/Fort_Prop_Close_Mono_Att
 bOverrideConcurrency=true
 maxDistance=1300
 soundClassObject=&Fort_Loot_Collect
 volumeMultiplier=0.1
 concurrencyOverrides {maxCount=1, resolutionRule=StopOldest}
}
