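# Sound cue for an electric zap, named for the player revive effect in gameplay.
# A single wave player supplies the Electricity_Zap sound to a Modulator at the root, which applies default random pitch and volume variation on each play.
# It belongs to the in-game SFX sound class and shows no attenuation settings, with a very large max distance.
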SoundCue /Game/Sounds/Fort_GamePlay_Sounds/Player/Zap_Electricity_Revive_Cue
root=Modulator_2
nodes [
 {id=Modulator_2, children [WavePlayer_8]}
 {id=WavePlayer_8, properties {soundWaveAssetPtr=&/Electricity_Zap}}
]
properties {maxDistance=2097152, soundClassObject=&Fort_SFX_InGame}
